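# Physics asset for the AntiZappo medium male player body cosmetic, covering a grenade hanging at the left pelvis.
# Capsule bodies form a chain from the grenade base through the grenade to its lock, with tight twist and narrow, asymmetric swing limits.
# A larger left thigh capsule, apparently a collider for the grenade, links to the base through a constraint with free swing and twist.
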PhysicsAsset /BRCosmetics/Characters/Player/Male/Medium/Bodies/M_MED_AntiZappo/Meshes/M_MED_AntiZappo_Physics
bodies {
 base_l_pelvis_granade [{shapeType=Capsule, length=0.505, radius=0.505, shapeName=base_l_pelvis_granade_capsule, rotation=<rot 0 -90 0>}]
 dyn_granade_lock [{shapeType=Capsule, length=2.3348, radius=2.2608, shapeName=dyn_granade_lock_capsule, center=<5 0 3.3>, rotation=<rot 90 0 0>}]
 dyn_l_pelvis_granade [{shapeType=Capsule, length=6, radius=4.3, shapeName=dyn_l_pelvis_granade_capsule, center=<5 0 3.3>, rotation=<rot 90 0 0>}]
 thigh_l [{shapeType=Capsule, length=22, radius=8.5, shapeName=thigh_l_capsule, center=<-3 0 0>, rotation=<rot 90 0 0>}]
}
constraints [
 {
  bone1Name=base_l_pelvis_granade
  bone2Name=thigh_l
  swing1LimitDegrees=45
  swing1Motion=Free
  swing2LimitDegrees=45
  swing2Motion=Free
  twistLimitDegrees=45
  twistMotion=Free
 }
 {bone1Name=dyn_l_pelvis_granade, bone2Name=base_l_pelvis_granade, swing1LimitDegrees=10, swing2LimitDegrees=30, twistLimitDegrees=1}
 {bone1Name=dyn_granade_lock, bone2Name=dyn_l_pelvis_granade, swing1LimitDegrees=5, swing2LimitDegrees=20, twistLimitDegrees=1}
]
